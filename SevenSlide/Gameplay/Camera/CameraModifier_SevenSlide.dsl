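; SevenSlide camera modifier that offsets view position and FOV of a Fortnite player pawn by movement speed and direction.
; Each update takes velocity from the actor or position deltas, then interpolates view-space offsets and FOV, with a timed lead-in forward offset.
; Aiming freezes side and vertical targets and adds forward and FOV bonuses.
; It disables on death, skips camera-mode-tagged pawns, and only sets, never reads, the post-boost flag.
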(event EventBlueprintAddedToCamera
  (params (Camera :type Player%20Camera%20Manager%20Object%20Reference))
  (Variables|Default|SetLastFrameRotation (Camera|GetCameraRotation Camera))
  (Variables|Default|SetModifierStartTime (Utilities|Time|GetGameTimeinSeconds))
  (bind _cameraowner (Variables|CameraModifier|GetCameraOwner)
    (:then
      (Utilities|IsValid (Game|Player|GetOwningPlayerController _cameraowner)
        (:Is%20Valid
          (Utilities|IsValid (Pawn|GetViewTarget (Game|Player|GetOwningPlayerController _cameraowner))
            (:Is%20Valid
              (Variables|Default|SetLastFramePosition_ (Transformation|GetActorLocation (Pawn|GetViewTarget (Game|Player|GetOwningPlayerController _cameraowner)))))))))
    (:else)))

(event Custom|RemovePostBoostMultilpier
  (Variables|Default|SetPostBoostingMultiplier false))

(event Custom|AddPostBoostMultiplier
  (Variables|Default|SetPostBoostingMultiplier true))

(fn BlueprintModifyCamera
  (params (DeltaTime :type Float%20%28single-precision%29) (ViewLocation :type Vector :default "0, 0, 0") (ViewRotation :type Rotator :default "0, 0, 0") (FOV :type Float%20%28single-precision%29))
  (bind _returnvalue (Math|Vector2D|MakeVector2D 15.0 9e+01))
  (bind _speed (LocalVariables|GetSpeed))
  (bind _output (Utilities|RemapFloat _speed "(X=150.000000,Y=1700.000000)" _returnvalue true))
  (bind _viewrotation (Variables|GetViewRotation))
  (bind _cameraowner (Variables|CameraModifier|GetCameraOwner)
    (:then
      (Utilities|IsValid (Game|Player|GetOwningPlayerController _cameraowner)
        (:Is%20Valid
          (bind _asplayer_pawn_athena (Utilities|Casting|CastToPlayerPawn_Athena (Pawn|GetViewTarget (Game|Player|GetOwningPlayerController _cameraowner)))
            (:then
              (bind _output_get (LocalVariables|SetViewTargetPawn _asplayer_pawn_athena))
              (if (FortPawn|IsDead _output_get)
                (CameraModifier|DisableModifier false)
                (else
                  (if (not (GameplayTags|HasMatchingGameplayTag (LocalVariables|GetViewTargetPawn) "(TagName=\"CameraController.CameraModes\")"))
                    (Utilities|FlowControl|Sequence
                      (:then_0
                        ; Cache Values
                        (LocalVariables|SetIsAiming (GameplayTags|HasAnyMatchingGameplayTags (LocalVariables|GetViewTargetPawn) (Variables|Default|GetIsAimingTags))))
                      (:then_1
                        ; Calculate Velocity
                        (bind _output_get_2 (LocalVariables|SetVelocity (Math|Vector|SelectVector (Transformation|GetVelocity (LocalVariables|GetViewTargetPawn)) (/ (- (Transformation|GetActorLocation (LocalVariables|GetViewTargetPawn)) (Variables|Default|GetLastFramePosition_)) (Math|Float|SelectFloat 1.0 (Variables|GetDeltaTime) (Math|Float|NearlyEqual(Float) (Variables|GetDeltaTime)))) (ScalableFloat|GetValueasBool (Variables|Default|GetUseActorVelocity)))))
                        (LocalVariables|SetDirection (Math|Vector|Normalize _output_get_2))
                        (LocalVariables|SetSpeed (Math|Vector|VectorLength _output_get_2))
                        (Variables|Default|SetLastFramePosition_ (Transformation|GetActorLocation (LocalVariables|GetViewTargetPawn))))
                      (:then_2
                        ; Cache view directions
                        (LocalVariables|SetViewForward (Math|Vector|GetForwardVector _viewrotation))
                        (LocalVariables|SetViewRight (Math|Vector|GetRightVector _viewrotation))
                        (LocalVariables|SetViewUp (Math|Vector|GetUpVector _viewrotation))
                        ; Calculate "straightness"
                        (LocalVariables|SetCurrentStraightness (Utilities|RemapFloat (Math|Float|Absolute(Float) (Math|Vector|DotProduct (LocalVariables|GetViewForward) (LocalVariables|GetDirection))) "(X=0.000000,Y=1.000000)" "(X=0.750000,Y=1.000000)" true))
                        ; Calculate "Is leading in"
                        (Variables|Default|SetLeadingIn (< (- (Utilities|Time|GetGameTimeinSeconds) (Variables|Default|GetModifierStartTime)) (Variables|Default|GetLeadInTime))))
                      (:then_3
                        ; Interpolate Aim-Down-Sights intensity
                        (Variables|Default|SetADS_Alpha (Math|Interpolation|FInterpTo (Variables|Default|GetADS_Alpha) (Math|Float|SelectFloat 1.0 0.0 (LocalVariables|GetIsAiming)) (Variables|GetDeltaTime) 6e+01))
                        ; Interpolate Lean
                        (Variables|Default|SetCurrentLean (Math|Interpolation|FInterpTo (Variables|Default|GetCurrentLean) 0.0 (Variables|GetDeltaTime) 7.0))
                        ; Interpolate FOV
                        (Variables|Default|SetCurrentFOV (Math|Interpolation|FInterpTo (Variables|Default|GetCurrentFOV) (Utilities|RemapFloat (LocalVariables|GetSpeed) "(X=150.000000,Y=1700.000000)" (Math|Vector2D|MakeVector2D (Variables|Default|GetScalingMovement_FOV_Min) (Variables|Default|GetScalingMovement_FOV_Max)) true) (Variables|GetDeltaTime) 2.5)))
                      (:then_4
                        ; Interpolate view space offset
                        (LocalVariables|GetIsAiming
                          (:then
                            _)
                          (:else
                            (Variables|Default|SetLastNonAimingForward _output)))
                        (Variables|Default|SetCurrentForward (Math|Interpolation|FInterpTo (Variables|Default|GetCurrentForward) (+ (Math|Float|SelectFloat (+ (Variables|Default|GetADSCameraForwardBonus) (Variables|Default|GetLastNonAimingForward)) _output (LocalVariables|GetIsAiming)) (Math|Float|SelectFloat (Variables|Default|GetLeadIn_ForwardOffset) 0.0 (Variables|Default|GetLeadingIn))) (Variables|GetDeltaTime) 6.0))
                        ;; If we are in ADS we don't want to change to hyper increase aiming fidelity
                        (LocalVariables|GetIsAiming
                          (:then)
                          (:else
                            (Variables|Default|SetCurrentRight (Math|Interpolation|FInterpTo (Variables|Default|GetCurrentRight) (Utilities|RemapFloat (Math|Vector|DotProduct (LocalVariables|GetVelocity) (LocalVariables|GetViewRight)) "(X=-1000.000000,Y=1000.000000)" (Class|EUWFunctionLibrary|LerpVector2D (Math|Vector2D|MakeVector2D -35.0 25.0) (Math|Vector2D|MakeVector2D -75.0 5.0) (Utilities|RemapFloat (LocalVariables|GetSpeed) "(X=450.000000,Y=1700.000000)" "(X=0.000000,Y=1.000000)" true)) true) (Variables|GetDeltaTime) (Utilities|RemapFloat (LocalVariables|GetSpeed) "(X=400.000000,Y=1500.000000)" "(X=3.000000,Y=7.000000)" true)))
                            (Variables|Default|SetCurrentUp (Math|Interpolation|FInterpTo (Variables|Default|GetCurrentUp) (Utilities|RemapFloat (LocalVariables|GetSpeed) "(X=400.000000,Y=1500.000000)" (Math|Vector2D|MakeVector2D -35.0 -35.0) true) (Variables|GetDeltaTime) (Utilities|RemapFloat (LocalVariables|GetSpeed) "(X=400.000000,Y=1500.000000)" "(X=4.000000,Y=6.000000)" true))))))
                      (:then_5
                        ; Interpolate current offset
                        (Variables|Default|SetCurrentOffset_Vec (Math|Interpolation|VInterpTo (Variables|Default|GetCurrentOffset_Vec) (Utilities|Operators|Add (* (LocalVariables|GetViewForward) (Variables|Default|GetCurrentForward)) (* (LocalVariables|GetViewRight) (+ (Variables|Default|GetCurrentRight) (* (LocalVariables|GetCurrentStraightness) -9e+01))) (* (LocalVariables|GetViewUp) (- (Variables|Default|GetCurrentUp) 18.0)) (* (Math|Vector|CrossProduct (LocalVariables|GetDirection) "0.000000,0.000000,1.000000") (Variables|Default|GetCurrentLean))) (Variables|GetDeltaTime) 8.0)))
                      (:then_6
                        ; Return result
                        (return (- (Variables|GetViewLocation) (Utilities|Operators|Add (Variables|Default|GetCurrentOffset_Vec) (* (LocalVariables|GetViewForward) -3e+01) (* (LocalVariables|GetViewUp) 3e+01) (* (LocalVariables|GetViewRight) -1e+01))) (Variables|GetViewRotation) (Utilities|Operators|Add (+ (Variables|GetFOV) (* (Variables|Default|GetCurrentFOV) (Variables|CameraModifier|GetAlpha))) (Math|Float|SelectFloat (ScalableFloat|GetValueasDouble (Variables|Default|GetADSFOVBonus)) 0.0 (LocalVariables|GetIsAiming)) (Math|Float|SelectFloat (Utilities|RemapFloat (Variables|Default|GetADS_Alpha) "(X=0.000000,Y=1.000000)" (Math|Vector2D|MakeVector2D 0.0 (ScalableFloat|GetValueasDouble (Variables|Default|GetADSFOVBonus))) true) 0.0 (and (not (LocalVariables|GetIsAiming)) (> (Variables|Default|GetADS_Alpha) 0))))))))))
              (return (Variables|GetViewLocation) (Variables|GetViewRotation) (Variables|GetFOV)))
            (:CastFailed
              _)))
        (:Is%20Not%20Valid
          _)))
    (:else
      _)))
